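# M_Ornaments is a lit material for ornaments in the winter front-end environment, with color variation and a gentle sway.
# Vertex color channels mask blends for color and scalar surface values, mixing fixed tints with a chart texture sampled by actor position.
# A switchable world-position rotation swings each actor about Z on a sine wave, deriving phase and speed from its X position.
# The same angle rotates the vertex normal to adjust shading.
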Material /Game/Environments/FrontEnd/Winter/Materials/M_Ornaments
graph {
 properties {pixelDepthOffsetMode=PDOM_Legacy, usedShadingModels=Default Lit}
 nodes [
  n1 = Constant3Vector() {constant=<rgba 0.0255 0.071 0.1458 0>}
  n2 = VertexColor()
  n3 = ActorPositionWS()
  n4 = TextureSampleParameter2D(UVs=Divide(
    A=MaterialFunctionCall!(f2, f3=@n3, f4) {materialFunction=/Game/Packages/Fortnite_UniqueMaterials_1/VectorComponentOps}.f3 Add
    B=ScalarParameter() {defaultValue=75, group=Colors, parameterName=ColorOffsetDistance}
   )) {group=Colors, parameterName=Color Chart 2, texture=/Game/Environments/FrontEnd/Winter/Textures/Ornaments_D}
  LinearInterpolate(
   A=LinearInterpolate(A=LinearInterpolate(A=@n1, Alpha=@n2.B), B=LinearInterpolate(A=@n4.RGB, Alpha=@n2.A), Alpha=@n2.G)
   B=Constant3Vector() {constant=<rgba 0.4896 0 0.0284 0>}
   Alpha=@n2.R
  )
  LinearInterpolate(
   A=LinearInterpolate(A=LinearInterpolate(Alpha=@n2.B) {constA=0.5, constB=0.35}, B=Constant() {r=0.2}, Alpha=@n2.G)
   B=Constant() {r=0.35}
   Alpha=@n2.R
  )
  LinearInterpolate(B=Constant3Vector() {constant=<rgba 0.0104 0 0.0006 0>}, Alpha=VertexColor().R)
  LinearInterpolate(A=LinearInterpolate(A=LinearInterpolate(Alpha=@n2.B), Alpha=@n2.G), Alpha=@n2.R)
  n5 = Constant3Vector() {constant=<rgba 0 0 1 0>}
  n6 = ActorPositionWS()
  n7 = MaterialFunctionCall(Float3=@n6) {materialFunction=&BreakOutFloat3Components}
  n8 = Frac(None=Divide(A=Add(A=@n7.R, B=@n7.R)) {constB=213.34})
  n9 = Add(A=Time(), B=Multiply(A=@n8, B=ScalarParameter() {defaultValue=109.2879, parameterName=Random}))
  n10 = LinearInterpolate(
   A=ScalarParameter() {defaultValue=0.0952, parameterName=RotSpeed}
   B=ScalarParameter() {defaultValue=0.03, parameterName=RotSpeed2}
   Alpha=@n8
  )
  n11 = Multiply(A=Sine(None=Multiply(A=@n9, B=@n10)), B=ScalarParameter() {defaultValue=0.1, parameterName=RotAngle})
  StaticSwitchParameter(True=RotateAboutAxis(NormalizedRotationAxis=@n5, RotationAngle=@n11, PivotPoint=@n6, Position=WorldPosition().XYZ), False=Constant3Vector()) {defaultValue=true, parameterName=UseRotation}
  n12 = VertexNormalWS()
  n13 = RotateAboutAxis(NormalizedRotationAxis=@n5, RotationAngle=@n11, PivotPoint=Constant3Vector(), Position=@n12)
  StaticSwitchParameter(
   True=Transform(None=Add(A=@n13, B=@n12)) {transformSourceType=TRANSFORMSOURCE_World, transformType=TRANSFORM_Tangent}
   False=Constant3Vector() {constant=<rgba 0 0 1 0>}
  ) {defaultValue=true, parameterName=UseRotation}
 ]
}
