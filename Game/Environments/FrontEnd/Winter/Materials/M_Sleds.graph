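# Surface material for sleds in the winter front-end environment, using Default Lit shading with no textures.
# A Color parameter, brown by default, is blended toward a darkened copy of itself by vertex green, then toward white by vertex red.
# Separate scalar interpolations, one driven by vertex red, are also present, though their material output connections are not shown.
# Painted vertex channels thus act as masks for the color variation.
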Material /Game/Environments/FrontEnd/Winter/Materials/M_Sleds
graph {
 properties {pixelDepthOffsetMode=PDOM_Legacy, usedShadingModels=Default Lit}
 nodes [
  n1 = VectorParameter() {parameterName=Color, defaultValue=<rgba 0.3073 0.1387 0.0416 0>}
  n2 = VertexColor()
  LinearInterpolate(A=LinearInterpolate(A=@n1.RGB, B=Multiply(A=@n1.RGB) {constB=0.35}, Alpha=@n2.G), Alpha=@n2.R)
  LinearInterpolate(Alpha=@n2.R)
  LinearInterpolate() {constA=0.75, constB=0.42}
 ]
}
